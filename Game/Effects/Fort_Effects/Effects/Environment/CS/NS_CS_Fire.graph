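# Large environmental fire Niagara system built from sprite emitters: air and ground flames, big embers, smoke (Smoke001; the original Smoke is disabled) and sparks.
# Curl noise and wind drive motion, with sparks adding gravity, vortex and time-of-day-scaled flicker.
# User.GlobalScaler scales spawn volumes, plus flame and ember rates, speed limits and material parameters.
# Camera-distance curves adjust spawn count, area, size and opacity, sleeping emitters beyond maximum range.
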NiagaraSystem /Game/Effects/Fort_Effects/Effects/Environment/CS/NS_CS_Fire
dependencies {
 usedDataInterfaces [&NiagaraDataInterfaceCurve, &NiagaraDataInterfaceCurlNoise, &NiagaraDataInterfaceColorCurve]
 usedDynamicInputs []
 usedModules [&SystemState]
}
emitters {
 FireAir {
  emitterUpdateScript [
   &EmitterState(
    Enable Distance Culling=true
    Max Distance Response=::SleepAndClearParticles
    MaxDistance=50000
    MinDistance=250
    Module.bApplySpawnCountScaleByDistance=true
    Module.bUseMaxDistance=true
    Module.bUseMinDistance=false
    Reset Age On Awaken=true
    Scalability Mode=::NewEnumerator1
    Scale Spawn Count=true
    Spawn Count Scale By Distance=DataInterface() {propertyValues {Curve {keys [<at 0 0.25>,<at 0.02 1>,<at 0.95 1>,<at 1 0>]}}}
   )
   &SpawnRate(SpawnRate=&Multiply_Float(A=15, B=$User.GlobalScaler), Use Spawn Probability=false)
  ]
  particleSpawnScript [
   /Niagara/Modules/Spawn/Initialization/InitializeParticle(
    Color=<rgba 1 0.15 0 1>
    Lifetime=&UniformRangedFloat(Minimum=1.5, Maximum=2)
    Module.Write Color=true
    Module.Write Lifetime=true
    Module.Write Mass=true
    Module.Write Material Random=true
    Module.Write Position=false
    Module.Write Scale=false
    Module.Write SpriteRotation=true
    Module.Write SpriteSize=true
    Sprite Rotation=&UniformRangedFloat(Minimum=-35, Maximum=45)
    Sprite Size=&Lerp_Vector2(A=<1000 1200>, B=<1000 1300>, Alpha=&RandomRangeFloat(Minimum=0, Maximum=1))
   )
   /Niagara/Modules/Spawn/Location/SphereLocation(
    MaskSpawnGroup=false
    Module.Override Randomness=false
    Non Uniform Scale=&Multiply_VectorByFloat(
     Vector=&Multiply_VectorByFloat(Vector=<1 1 1>, Float=$User.GlobalScaler)
     Float=/Niagara/DynamicInputs/Scalability/ScalabilityDistanceBasedFloat(
      Curve:NiagaraDataInterfaceCurve=DataInterface() {propertyValues {Curve {keys [<at 0 0.25>, {arriveTangent=0.5882, time=0.1, value=0.5}, <at 0.85 1>, <at 0.95 0>]}}}
      CurveIndex:NiagaraFloat=$Emitter.Scalability.DistanceFraction
      Bias:NiagaraFloat=0
      Scale:NiagaraFloat=2
     )
    )
    Offset=<-125 0 0>
    Sphere Radius=200
    Surface Only=false
   )
   &AddVelocity(Velocity=&UniformRangedVector(Minimum=<0 0 400>, Maximum=<0 0 600>))
   &ShapeLocation(
    Cylinder Height=250
    Cylinder Radius=250
    MaskSpawnGroup=false
    Offset=<0 0 125>
    Offset Mode=::NewEnumerator0
    Override Seed=false
    Shape Primitive=::NewEnumerator1
    Surface Only=false
    UseOwnerScale=true
   ) {enabled=false}
   /Niagara/Modules/Spawn/Location/V2/SphereLocation(MaskSpawnGroup=false, Override Seed=false, Surface Only=false, UseNonUniformScale=false, UseOffset=false, UseSphereOrientation=false) {module=SphereLocation001}
  ]
  particleUpdateScript [
   &ParticleState() {module=UpdateAge}
   &AccelerationForce(Acceleration=<0 0 50>)
   /Niagara/Modules/Update/Forces/CurlNoiseForce(Noise Frequency=0.1, Noise Strength=500, Pan Noise=true, Pan Noise Field=<0 0 0.25>)
   &WindForce(Wind Speed=<100 0 0>)
   &SolveForcesAndVelocity(
    Clamp Velocity=true
    Limit Acceleration=false
    Speed Limit:NiagaraFloat=&Multiply_Float(A=1500, B=$User.GlobalScaler)
   )
   &ScaleSpriteSize(Scale Factor=&Vector2DFromFloat(Value=&FloatFromCurve(
      FloatCurve=DataInterface() {propertyValues {Curve {keys [
          {arriveTangent=3.1239, interpMode=RCIM_Cubic, leaveTangent=3.1239, tangentMode=RCTM_User, time=0, value=0.5}
          {arriveTangent=0.1, interpMode=RCIM_Cubic, leaveTangent=0.1, time=0.1, value=0.975}
          <at 1 0.6 RCIM_Cubic>
         ]}}}
      Scale Curve=&ScalabilityDistanceBasedFloat(
       Curve:NiagaraDataInterfaceCurve=DataInterface() {propertyValues {Curve {keys [<at 0 0.25>, {arriveTangent=0.5882, time=0.1, value=0.5}, <at 0.8 1>, <at 0.9 0>]}}}
       CurveIndex:NiagaraFloat=$Emitter.Scalability.DistanceFraction
       Bias:NiagaraFloat=0
       Scale:NiagaraFloat=1.5
      )
     ))) {module=SpriteSizeScale}
   &SpriteRotationRate(Rotation Rate=&UniformRangedFloat(Minimum=-100, Maximum=100))
   &DynamicMaterialParameters(
    Index 0 Param 1=&RandomRangeFloat(Minimum=0, Maximum=1)
    Index 0 Param 2=&RandomRangeFloat(Minimum=0, Maximum=1)
    Index 0 Param 3=&Multiply_Float(
     A=&RandomRangeFloat(Minimum=0.5, Maximum=1)
     B=&ClampFloat(Float=$User.GlobalScaler, Min=0.25, Max=1.25)
    )
    Index 0 Param 4=&Multiply_Float(
     A=&RandomRangeFloat(Minimum=0.5, Maximum=1)
     B=&ClampFloat(Float=$User.GlobalScaler, Min=0.25, Max=1.25)
    )
    Index 1 Param 1=&FloatFromCurve(FloatCurve=DataInterface() {propertyValues {Curve {keys [
         {interpMode=RCIM_Cubic, tangentMode=RCTM_User, time=0, value=1}
         {interpMode=RCIM_Cubic, tangentMode=RCTM_User, time=0.2, value=0.05}
         {interpMode=RCIM_Cubic, tangentMode=RCTM_User, time=0.8, value=0.075}
         {interpMode=RCIM_Cubic, tangentMode=RCTM_User, time=1, value=1.5}
        ]}}})
    Index 1 Param 2=&Multiply_Float(
     A=&RandomRangeFloat(Minimum=-0.25, Maximum=0.25)
     B=&ClampFloat(Float=$User.GlobalScaler, Min=-0.25)
    )
    Index 1 Param 3=&Multiply_Float(
     A=&RandomRangeFloat(Minimum=0.6, Maximum=0.8)
     B=&ClampFloat(Float=$User.GlobalScaler, Min=0.2)
    )
    Index 1 Param 4=&RandomRangeFloat(Minimum=5, Maximum=10)
    Index 2 Param 1=&Multiply_Float(A=&RandomRangeFloat(Minimum=1000, Maximum=1500), B=$User.GlobalScaler)
    Index 2 Param 2=&RandomRangeFloat(Minimum=0.4, Maximum=0.5)
    Module.Param0WriteEnabled=true
    Module.Param0WriteEnabled1=true
    Module.Param0WriteEnabled2=true
    Module.Param1WriteEnabled=true
    Module.Param1WriteEnabled1=true
    Module.Param1WriteEnabled2=true
    Module.Param2WriteEnabled=true
    Module.Param2WriteEnabled1=true
    Module.Param2WriteEnabled2=true
    Module.Param3WriteEnabled=true
    Module.Param3WriteEnabled1=true
    Module.Param3WriteEnabled2=true
    Write Parameter Index 1=true
    Write Parameter Index 2=true
   )
   &ScaleColor(
    Scale Alpha=&Multiply_Float(
     A=&FloatFromCurve(FloatCurve=DataInterface() {propertyValues {Curve {keys [{interpMode=RCIM_Cubic, tangentMode=RCTM_User, time=0, value=0}, {interpMode=RCIM_Cubic, tangentMode=RCTM_User, time=0.2, value=1}, <at 1 0>]}}})
     B=&ScalabilityDistanceBasedFloat(
      Curve:NiagaraDataInterfaceCurve=DataInterface() {propertyValues {Curve {keys [<at 0 0>, {arriveTangent=1.1111, leaveTangent=1.1111, time=0.05, value=1}, <at 0.7 1>, <at 0.9 0>]}}}
      CurveIndex:NiagaraFloat=$Emitter.Scalability.DistanceFraction
      Bias:NiagaraFloat=0
      Scale:NiagaraFloat=1
     )
    )
    ScaleA=true
    ScaleRGB=true
   )
  ]
  rendererClasses [&NiagaraSpriteRendererProperties]
  renderers [{rendererClass=&NiagaraSpriteRendererProperties}]
 }
 FireGround {
  simTarget=GPUComputeSim
  emitterUpdateScript [
   &EmitterState(
    Enable Distance Culling=true
    Max Distance Response=::SleepAndClearParticles
    MaxDistance=50000
    MinDistance=250
    Module.bApplySpawnCountScaleByDistance=true
    Module.bUseMaxDistance=true
    Module.bUseMinDistance=false
    Reset Age On Awaken=true
    Scalability Mode=::NewEnumerator1
    Scale Spawn Count=true
    Spawn Count Scale By Distance=DataInterface() {propertyValues {Curve {keys [<at 0 0.25>,<at 0.02 1>,<at 0.95 1>,<at 1 0>]}}}
   )
   &SpawnRate(SpawnRate=&Multiply_Float(A=8, B=$User.GlobalScaler), Use Spawn Probability=false)
  ]
  particleSpawnScript [
   &InitializeParticle(
    Color=<rgba 1 0.15 0 1>
    Lifetime=&UniformRangedFloat(Minimum=1.25, Maximum=2)
    Module.Write Color=true
    Module.Write Lifetime=true
    Module.Write Mass=true
    Module.Write Material Random=true
    Module.Write Position=false
    Module.Write Scale=false
    Module.Write SpriteRotation=true
    Module.Write SpriteSize=true
    Sprite Rotation=&UniformRangedFloat(Minimum=-35, Maximum=45)
    Sprite Size=&Lerp_Vector2(A=<400 500>, B=<400 500>, Alpha=&RandomRangeFloat(Minimum=0, Maximum=1))
   )
   /Niagara/Modules/Spawn/Location/SphereLocation(
    MaskSpawnGroup=false
    Module.Override Randomness=false
    Non Uniform Scale=&Multiply_VectorByFloat(
     Vector=&Multiply_VectorByFloat(Vector=<1 1 0.1>, Float=$User.GlobalScaler)
     Float=&ScalabilityDistanceBasedFloat(
      Curve:NiagaraDataInterfaceCurve=DataInterface() {propertyValues {Curve {keys [<at 0 0.25>, {arriveTangent=0.5882, time=0.1, value=0.5}, <at 0.85 1>, <at 0.95 0>]}}}
      CurveIndex:NiagaraFloat=$Emitter.Scalability.DistanceFraction
      Bias:NiagaraFloat=0
      Scale:NiagaraFloat=2
     )
    )
    Offset=<-250 0 0>
    Sphere Radius=200
    Surface Only=false
   )
   &AddVelocity(Velocity=&UniformRangedVector(Minimum=<0 0 100>, Maximum=<0 0 25>))
   &ShapeLocation(
    Cylinder Height=250
    Cylinder Radius=250
    MaskSpawnGroup=false
    Offset=<0 0 125>
    Offset Mode=::NewEnumerator0
    Override Seed=false
    Shape Primitive=::NewEnumerator1
    Surface Only=false
    UseOwnerScale=true
   ) {enabled=false}
   /Niagara/Modules/Spawn/Location/V2/SphereLocation(MaskSpawnGroup=false, Override Seed=false, Surface Only=false, UseNonUniformScale=false, UseOffset=false, UseSphereOrientation=false) {module=SphereLocation001}
  ]
  particleUpdateScript [
   &ParticleState() {module=UpdateAge}
   &AccelerationForce(Acceleration=<0 0 50>)
   &CurlNoiseForce(Noise Frequency=0.1, Noise Strength=50, Pan Noise=true, Pan Noise Field=<0 0 0.25>)
   &WindForce(Wind Speed=<1 0 0>)
   &SolveForcesAndVelocity(
    Clamp Velocity=true
    Limit Acceleration=false
    Speed Limit:NiagaraFloat=&Multiply_Float(A=1500, B=$User.GlobalScaler)
   )
   &ScaleSpriteSize(Scale Factor=&Vector2DFromFloat(Value=&FloatFromCurve(
      FloatCurve=DataInterface() {propertyValues {Curve {keys [
          {arriveTangent=3.1239, interpMode=RCIM_Cubic, leaveTangent=3.1239, tangentMode=RCTM_User, time=0, value=0.5}
          {arriveTangent=0.1, interpMode=RCIM_Cubic, leaveTangent=0.1, time=0.1, value=0.975}
          <at 1 0.6 RCIM_Cubic>
         ]}}}
      Scale Curve=&ScalabilityDistanceBasedFloat(
       Curve:NiagaraDataInterfaceCurve=DataInterface() {propertyValues {Curve {keys [<at 0 0.25>, {arriveTangent=0.5882, time=0.1, value=1}, <at 0.8 1>, <at 0.9 0>]}}}
       CurveIndex:NiagaraFloat=$Emitter.Scalability.DistanceFraction
       Bias:NiagaraFloat=0
       Scale:NiagaraFloat=1.5
      )
     ))) {module=SpriteSizeScale}
   &SpriteRotationRate(Rotation Rate=&UniformRangedFloat(Minimum=-100, Maximum=100))
   &DynamicMaterialParameters(
    Index 0 Param 1=&RandomRangeFloat(Minimum=0, Maximum=1)
    Index 0 Param 2=&RandomRangeFloat(Minimum=0, Maximum=1)
    Index 0 Param 3=&Multiply_Float(
     A=&RandomRangeFloat(Minimum=0.5, Maximum=1)
     B=&ClampFloat(Float=$User.GlobalScaler, Min=0.25, Max=1.25)
    )
    Index 0 Param 4=&Multiply_Float(
     A=&RandomRangeFloat(Minimum=0.5, Maximum=1)
     B=&ClampFloat(Float=$User.GlobalScaler, Min=0.25, Max=1.25)
    )
    Index 1 Param 1=&FloatFromCurve(FloatCurve=DataInterface() {propertyValues {Curve {keys [
         {interpMode=RCIM_Cubic, tangentMode=RCTM_User, time=0, value=1}
         {interpMode=RCIM_Cubic, tangentMode=RCTM_User, time=0.2, value=0.05}
         {interpMode=RCIM_Cubic, tangentMode=RCTM_User, time=0.8, value=0.075}
         {interpMode=RCIM_Cubic, tangentMode=RCTM_User, time=1, value=1.5}
        ]}}})
    Index 1 Param 2=&Multiply_Float(
     A=&RandomRangeFloat(Minimum=-0.15, Maximum=0.15)
     B=&ClampFloat(Float=$User.GlobalScaler, Min=-0.25)
    )
    Index 1 Param 3=&Multiply_Float(
     A=&RandomRangeFloat(Minimum=0.6, Maximum=1)
     B=&ClampFloat(Float=$User.GlobalScaler, Min=0.2)
    )
    Index 1 Param 4=&RandomRangeFloat(Minimum=5, Maximum=10)
    Index 2 Param 1=&Multiply_Float(A=&RandomRangeFloat(Minimum=1000, Maximum=1500), B=$User.GlobalScaler)
    Index 2 Param 2=&RandomRangeFloat(Minimum=0.4, Maximum=0.5)
    Module.Param0WriteEnabled=true
    Module.Param0WriteEnabled1=true
    Module.Param0WriteEnabled2=true
    Module.Param1WriteEnabled=true
    Module.Param1WriteEnabled1=true
    Module.Param1WriteEnabled2=true
    Module.Param2WriteEnabled=true
    Module.Param2WriteEnabled1=true
    Module.Param2WriteEnabled2=true
    Module.Param3WriteEnabled=true
    Module.Param3WriteEnabled1=true
    Module.Param3WriteEnabled2=true
    Write Parameter Index 1=true
    Write Parameter Index 2=true
   )
   &ScaleColor(
    Scale Alpha=&Multiply_Float(
     A=&FloatFromCurve(FloatCurve=DataInterface() {propertyValues {Curve {keys [{interpMode=RCIM_Cubic, tangentMode=RCTM_User, time=0, value=0}, {interpMode=RCIM_Cubic, tangentMode=RCTM_User, time=0.2, value=1}, <at 1 0>]}}})
     B=&ScalabilityDistanceBasedFloat(
      Curve:NiagaraDataInterfaceCurve=DataInterface() {propertyValues {Curve {keys [<at 0 0>, {arriveTangent=1.1111, leaveTangent=1.1111, time=0.15, value=1}, <at 0.7 1>, <at 0.9 0>]}}}
      CurveIndex:NiagaraFloat=$Emitter.Scalability.DistanceFraction
      Bias:NiagaraFloat=0
      Scale:NiagaraFloat=1
     )
    )
    ScaleA=true
    ScaleRGB=true
   )
  ]
  rendererClasses [&NiagaraSpriteRendererProperties]
  renderers [{rendererClass=&NiagaraSpriteRendererProperties}]
 }
 Fire_BigEmbers {
  simTarget=GPUComputeSim
  emitterUpdateScript [
   &EmitterState(
    Enable Distance Culling=true
    Max Distance Response=::SleepAndClearParticles
    MaxDistance=50000
    MinDistance=250
    Module.bApplySpawnCountScaleByDistance=true
    Module.bUseMaxDistance=true
    Module.bUseMinDistance=false
    Reset Age On Awaken=true
    Scalability Mode=::NewEnumerator1
    Scale Spawn Count=true
    Spawn Count Scale By Distance=DataInterface() {propertyValues {Curve {keys [<at 0 0.25>,<at 0.02 1>,<at 0.95 1>,<at 1 0>]}}}
   )
   &SpawnRate(SpawnRate=&Multiply_Float(A=10, B=$User.GlobalScaler), Use Spawn Probability=false)
  ]
  particleSpawnScript [
   &InitializeParticle(
    Color=<rgba 1 0.15 0 1>
    Lifetime=&UniformRangedFloat(Minimum=3, Maximum=5)
    Module.Write Color=true
    Module.Write Lifetime=true
    Module.Write Mass=true
    Module.Write Material Random=true
    Module.Write Position=false
    Module.Write Scale=false
    Module.Write SpriteRotation=true
    Module.Write SpriteSize=true
    Sprite Rotation=&UniformRangedFloat(Minimum=-360, Maximum=360)
    Sprite Size=&Lerp_Vector2(A=<25 50>, B=<35 80>, Alpha=&RandomRangeFloat(Minimum=0, Maximum=1))
   )
   /Niagara/Modules/Spawn/Location/SphereLocation(
    MaskSpawnGroup=false
    Module.Override Randomness=false
    Non Uniform Scale=&Multiply_VectorByFloat(
     Vector=&Multiply_VectorByFloat(Vector=<1 1 1.25>, Float=$User.GlobalScaler)
     Float=&ScalabilityDistanceBasedFloat(
      Curve:NiagaraDataInterfaceCurve=DataInterface() {propertyValues {Curve {keys [<at 0 0.25>, {arriveTangent=0.5882, time=0.1, value=0.5}, <at 0.85 1>, <at 0.95 0>]}}}
      CurveIndex:NiagaraFloat=$Emitter.Scalability.DistanceFraction
      Bias:NiagaraFloat=0
      Scale:NiagaraFloat=2
     )
    )
    Offset=<0 0 1500>
    Sphere Radius=600
    Surface Only=false
   )
   &AddVelocity(Velocity=&UniformRangedVector(Minimum=<0 0 1500>, Maximum=<0 0 1800>))
   &ShapeLocation(
    Cylinder Height=250
    Cylinder Radius=250
    MaskSpawnGroup=false
    Offset=<0 0 125>
    Offset Mode=::NewEnumerator0
    Override Seed=false
    Shape Primitive=::NewEnumerator1
    Surface Only=false
    UseOwnerScale=true
   ) {enabled=false}
   /Niagara/Modules/Spawn/Location/V2/SphereLocation(MaskSpawnGroup=false, Override Seed=false, Surface Only=false, UseNonUniformScale=false, UseOffset=false, UseSphereOrientation=false) {module=SphereLocation001}
  ]
  particleUpdateScript [
   &ParticleState() {module=UpdateAge}
   &AccelerationForce(Acceleration=<0 0 50>)
   &CurlNoiseForce(Noise Frequency=1, Noise Strength=5000, Pan Noise=true, Pan Noise Field=<0 0 0.5>)
   &WindForce(Wind Speed=<250 0 0>)
   &SolveForcesAndVelocity(
    Clamp Velocity=true
    Limit Acceleration=false
    Speed Limit:NiagaraFloat=&Multiply_Float(A=1500, B=$User.GlobalScaler)
   )
   &ScaleSpriteSize(Scale Factor=&Vector2DFromFloat(Value=&FloatFromCurve(
      FloatCurve=DataInterface() {propertyValues {Curve {keys [
          {arriveTangent=3.1239, interpMode=RCIM_Cubic, leaveTangent=3.1239, tangentMode=RCTM_User, time=0, value=0.5}
          {arriveTangent=0.1, interpMode=RCIM_Cubic, leaveTangent=0.1, time=0.2457, value=0.9993}
          <at 1 0.6 RCIM_Cubic>
         ]}}}
      Scale Curve=&ScalabilityDistanceBasedFloat(
       Curve:NiagaraDataInterfaceCurve=DataInterface() {propertyValues {Curve {keys [<at 0 0.25>, {arriveTangent=0.5882, time=0.1, value=0.5}, <at 0.75 1>, <at 0.85 0>]}}}
       CurveIndex:NiagaraFloat=$Emitter.Scalability.DistanceFraction
       Bias:NiagaraFloat=0
       Scale:NiagaraFloat=1.5
      )
     ))) {module=SpriteSizeScale}
   &SpriteRotationRate(Rotation Rate=&UniformRangedFloat(Minimum=-100, Maximum=100))
   &DynamicMaterialParameters(
    Index 0 Param 1=&RandomRangeFloat(Minimum=0, Maximum=1)
    Index 0 Param 2=&RandomRangeFloat(Minimum=0, Maximum=1)
    Index 0 Param 3=&Multiply_Float(
     A=&RandomRangeFloat(Minimum=0.5, Maximum=1)
     B=&ClampFloat(Float=$User.GlobalScaler, Min=0.25, Max=1.25)
    )
    Index 0 Param 4=&Multiply_Float(
     A=&RandomRangeFloat(Minimum=0.5, Maximum=1)
     B=&ClampFloat(Float=$User.GlobalScaler, Min=0.25, Max=1.25)
    )
    Index 1 Param 1=&FloatFromCurve(FloatCurve=DataInterface() {propertyValues {Curve {keys [
         {interpMode=RCIM_Cubic, tangentMode=RCTM_User, time=0, value=1}
         {interpMode=RCIM_Cubic, tangentMode=RCTM_User, time=0.2, value=0.05}
         {interpMode=RCIM_Cubic, tangentMode=RCTM_User, time=0.8, value=0.075}
         {interpMode=RCIM_Cubic, tangentMode=RCTM_User, time=1, value=1.5}
        ]}}})
    Index 1 Param 2=&Multiply_Float(
     A=&RandomRangeFloat(Minimum=-0.25, Maximum=0.25)
     B=&ClampFloat(Float=$User.GlobalScaler, Min=-0.25)
    )
    Index 1 Param 3=&Multiply_Float(
     A=&RandomRangeFloat(Minimum=0.6, Maximum=0.8)
     B=&ClampFloat(Float=$User.GlobalScaler, Min=0.2)
    )
    Index 1 Param 4=&RandomRangeFloat(Minimum=5, Maximum=10)
    Index 2 Param 1=&Multiply_Float(A=&RandomRangeFloat(Minimum=1000, Maximum=1500), B=$User.GlobalScaler)
    Index 2 Param 2=&RandomRangeFloat(Minimum=0.4, Maximum=0.5)
    Module.Param0WriteEnabled=true
    Module.Param0WriteEnabled1=true
    Module.Param0WriteEnabled2=true
    Module.Param1WriteEnabled=true
    Module.Param1WriteEnabled1=true
    Module.Param1WriteEnabled2=true
    Module.Param2WriteEnabled=true
    Module.Param2WriteEnabled1=true
    Module.Param2WriteEnabled2=true
    Module.Param3WriteEnabled=true
    Module.Param3WriteEnabled1=true
    Module.Param3WriteEnabled2=true
    Write Parameter Index 1=true
    Write Parameter Index 2=true
   )
   &ScaleColor(
    Scale Alpha=&FloatFromCurve(
     FloatCurve=DataInterface() {propertyValues {Curve {keys [{interpMode=RCIM_Cubic, tangentMode=RCTM_User, time=0, value=0}, {interpMode=RCIM_Cubic, tangentMode=RCTM_User, time=0.2, value=1}, <at 1 0>]}}}
     Scale Curve=&ScalabilityDistanceBasedFloat(
      Curve:NiagaraDataInterfaceCurve=DataInterface() {propertyValues {Curve {keys [
          {interpMode=RCIM_Cubic, tangentMode=RCTM_User, time=0, value=0}
          {interpMode=RCIM_Cubic, tangentMode=RCTM_User, time=0.15, value=1}
          {interpMode=RCIM_Cubic, tangentMode=RCTM_User, time=0.7, value=1}
          <at 0.9 0>
         ]}}}
      CurveIndex:NiagaraFloat=$Emitter.Scalability.DistanceFraction
      Bias:NiagaraFloat=0
      Scale:NiagaraFloat=1
     )
    )
    ScaleA=true
    ScaleRGB=true
   )
   /Niagara/Modules/Update/Camera/MaintainInCameraParticleScale()
  ]
  rendererClasses [&NiagaraSpriteRendererProperties]
  renderers [{rendererClass=&NiagaraSpriteRendererProperties}]
 }
 Smoke {
  bEnabled=false
  emitterUpdateScript [
   &EmitterState(
    Enable Distance Culling=true
    Max Distance Response=::SleepAndClearParticles
    MaxDistance=50000
    MinDistance=250
    Module.bApplySpawnCountScaleByDistance=true
    Module.bUseMaxDistance=true
    Module.bUseMinDistance=false
    Reset Age On Awaken=true
    Scalability Mode=::NewEnumerator1
    Scale Spawn Count=true
    Spawn Count Scale By Distance=DataInterface() {propertyValues {Curve {keys [<at 0 0.25>,<at 0.02 1>,<at 0.8 1>,<at 1 0>]}}}
   )
   &SpawnRate(SpawnRate=4, Use Spawn Probability=false)
  ]
  particleSpawnScript [
   /Niagara/Modules/Spawn/Initialization/V2/InitializeParticle(
    Lifetime Max=5
    Lifetime Min=3
    Lifetime Mode=::NewEnumerator1
    Sprite Rotation Angle=&RandomRangeFloat(Minimum=-0.05, Maximum=0.05)
    Sprite Rotation Mode=::NewEnumerator1
    Sprite Size Mode=::NewEnumerator1
    Uniform Sprite Size Max=2500
    Uniform Sprite Size Min=2000
    UsePositionOffset=false
    Write Material Random=true
   )
   /Niagara/Modules/Spawn/Location/SphereLocation(
    MaskSpawnGroup=false
    Module.Override Randomness=false
    Non Uniform Scale=&Multiply_VectorByFloat(
     Vector=&Multiply_VectorByFloat(Vector=<1 1 0.25>, Float=$User.GlobalScaler)
     Float=&ScalabilityDistanceBasedFloat(
      Curve:NiagaraDataInterfaceCurve=DataInterface() {propertyValues {Curve {keys [<at 0 0.25>, {arriveTangent=0.5882, time=0.1, value=0.5}, <at 0.85 1>, <at 0.95 0>]}}}
      CurveIndex:NiagaraFloat=$Emitter.Scalability.DistanceFraction
      Bias:NiagaraFloat=0
      Scale:NiagaraFloat=2
     )
    )
    Sphere Radius=400
    Surface Only=false
   )
   &AddVelocity(Velocity=&UniformRangedVector(Minimum=<0 0 800>, Maximum=<0 0 1200>))
  ]
  particleUpdateScript [
   &ParticleState()
   &ScaleColor(
    Linear Color Curve=DataInterface() {propertyValues {
      AlphaCurve {keys [<at 0 1>,<at 1 0>]}
      BlueCurve {keys [<at 0 0.01>,<at 0.5465 0.025>]}
      GreenCurve {keys [<at 0 0.2>,<at 0.5465 0.027>]}
      RedCurve {keys [<at 0 1>,<at 0.5465 0.05>]}
     }}
    Scale Mode=::NewEnumerator2
   )
   /SetVariables(Particles.ScaledColor:LinearColor=&MultiplyLinearColor(
     Linear Color=$Particles.Color
     Scale Factor=&MakeVector4(
      X=2
      Y=2
      Z=2
      W=&Multiply_Float(
       A=&FloatFromCurve(
        FloatCurve=DataInterface() {propertyValues {Curve {keys [
            {interpMode=RCIM_Cubic, tangentMode=RCTM_User, time=0, value=0}
            {interpMode=RCIM_Cubic, tangentMode=RCTM_User, time=0.1, value=0.5}
            <at 0.8 1.125>
            {tangentMode=RCTM_User, time=1, value=0}
           ]}}}
        Scale Curve=&FNTimeOfDayScalar(Night Scale=1, Morning Scale=0.25, Evening Scale=0.25)
       )
       B=&ScalabilityDistanceBasedFloat(
        Curve:NiagaraDataInterfaceCurve=DataInterface() {propertyValues {Curve {keys [
            {interpMode=RCIM_Cubic, tangentMode=RCTM_User, time=0, value=0}
            {interpMode=RCIM_Cubic, tangentMode=RCTM_User, time=0.05, value=1}
            {interpMode=RCIM_Cubic, tangentMode=RCTM_User, time=0.7, value=1}
            <at 0.9 0>
           ]}}}
        CurveIndex:NiagaraFloat=$Emitter.Scalability.DistanceFraction
        Bias:NiagaraFloat=0
        Scale:NiagaraFloat=0.75
       )
      )
     )
    )) {isSetParameters=true, module=SetVariables_#1}
   &ScaleSpriteSize(Scale Factor=&Vector2DFromFloat(Value=&FloatFromCurve(
      FloatCurve=DataInterface() {propertyValues {Curve {keys [
          {arriveTangent=1.8751, interpMode=RCIM_Cubic, leaveTangent=1.8751, tangentMode=RCTM_User, time=0, value=0.2}
          {arriveTangent=1.0155, interpMode=RCIM_Cubic, leaveTangent=1.0155, tangentMode=RCTM_User, time=0.4, value=1.05}
          {arriveTangent=0.0421, interpMode=RCIM_Cubic, leaveTangent=0.0421, tangentMode=RCTM_User, time=1, value=1.25}
         ]}}}
      Scale Curve=&ScalabilityDistanceBasedFloat(
       Curve:NiagaraDataInterfaceCurve=DataInterface() {propertyValues {Curve {keys [<at 0 0.25>, {arriveTangent=0.5882, time=0.1, value=0.5}, <at 0.7 1>, <at 0.8 0>]}}}
       CurveIndex:NiagaraFloat=$Emitter.Scalability.DistanceFraction
       Bias:NiagaraFloat=0
       Scale:NiagaraFloat=1.5
      )
     )))
   &AccelerationForce()
   &CurlNoiseForce(Noise Frequency=0.1, Noise Strength=500, Pan Noise=true, Pan Noise Field=<0 0 0.25>)
   &WindForce(Wind Speed=<500 0 0>)
   &SolveForcesAndVelocity(Clamp Velocity=false, Limit Acceleration=false)
   &SpriteRotationRate(Rotation Rate=&UniformRangedFloat(Minimum=-15, Maximum=15))
   &DynamicMaterialParameters(
    Index 0 Param 1=&FloatFromCurve(FloatCurve=DataInterface() {propertyValues {Curve {keys [
         {arriveTangent=-7.0802, interpMode=RCIM_Cubic, leaveTangent=-7.0802, tangentMode=RCTM_User, time=0, value=2}
         {interpMode=RCIM_Cubic, tangentMode=RCTM_User, time=0.5, value=0.05}
         {interpMode=RCIM_Cubic, tangentMode=RCTM_User, time=0.6, value=0.1}
         <at 1 2>
        ]}}})
    Index 0 Param 2=&FloatFromCurve(FloatCurve=DataInterface() {propertyValues {Curve {keys [
         {interpMode=RCIM_Cubic, tangentMode=RCTM_User, time=0, value=0}
         {interpMode=RCIM_Cubic, tangentMode=RCTM_User, time=0.5, value=1}
         {tangentMode=RCTM_User, time=1, value=0}
        ]}}})
    Index 0 Param 3=&RandomRangeFloat(Minimum=0.6, Maximum=0.8)
    Index 1 Param 1=150
    Module.Param0WriteEnabled=true
    Module.Param0WriteEnabled1=true
    Module.Param1WriteEnabled=true
    Module.Param1WriteEnabled1=true
    Module.Param2WriteEnabled=true
    Module.Param2WriteEnabled1=true
    Module.Param3WriteEnabled=true
    Module.Param3WriteEnabled1=true
    Write Parameter Index 1=true
   )
  ]
  rendererClasses [&NiagaraSpriteRendererProperties]
  renderers [{rendererClass=&NiagaraSpriteRendererProperties}]
 }
 Smoke001 {
  emitterUpdateScript [
   &EmitterState(
    Enable Distance Culling=true
    Max Distance Response=::SleepAndClearParticles
    MaxDistance=50000
    MinDistance=250
    Module.bApplySpawnCountScaleByDistance=true
    Module.bUseMaxDistance=true
    Module.bUseMinDistance=false
    Reset Age On Awaken=true
    Scalability Mode=::NewEnumerator1
    Scale Spawn Count=true
    Spawn Count Scale By Distance=DataInterface() {propertyValues {Curve {keys [<at 0 0.25>,<at 0.02 1>,<at 0.8 1>,<at 1 0>]}}}
   )
   &SpawnRate(SpawnRate=4, Use Spawn Probability=false)
  ]
  particleSpawnScript [
   &InitializeParticle(
    Lifetime Max=4
    Lifetime Min=2.5
    Lifetime Mode=::NewEnumerator1
    Sprite Rotation Angle=&RandomRangeFloat(Minimum=-0.15, Maximum=0.15)
    Sprite Rotation Mode=::NewEnumerator1
    Sprite Size Mode=::NewEnumerator1
    Uniform Sprite Size Max=2500
    Uniform Sprite Size Min=2000
    UsePositionOffset=false
    Write Material Random=true
   )
   &SphereLocation(
    MaskSpawnGroup=false
    Module.Override Randomness=false
    Non Uniform Scale=&Multiply_VectorByFloat(
     Vector=&Multiply_VectorByFloat(Vector=<1 1 0.25>, Float=$User.GlobalScaler)
     Float=&ScalabilityDistanceBasedFloat(
      Curve:NiagaraDataInterfaceCurve=DataInterface() {propertyValues {Curve {keys [<at 0 0.25>, {arriveTangent=0.5882, time=0.1, value=0.5}, <at 0.85 1>, <at 0.95 0>]}}}
      CurveIndex:NiagaraFloat=$Emitter.Scalability.DistanceFraction
      Bias:NiagaraFloat=0
      Scale:NiagaraFloat=2
     )
    )
    Sphere Radius=400
    Surface Only=false
   )
   &AddVelocity(Velocity=&UniformRangedVector(Minimum=<0 0 800>, Maximum=<0 0 1200>))
  ]
  particleUpdateScript [
   &ParticleState()
   &ScaleColor(
    Linear Color Curve=DataInterface() {propertyValues {
      AlphaCurve {keys [<at 0 1>,<at 0.8291 1>,<at 1 0>]}
      BlueCurve {keys [<at 0 0.01>,<at 0.5465 0.0476>]}
      GreenCurve {keys [<at 0 0.2>,<at 0.5465 0.0729>]}
      RedCurve {keys [<at 0 1>,<at 0.5465 0.125>]}
     }}
    Scale Mode=::NewEnumerator2
   )
   /SetVariables(Particles.ScaledColor:LinearColor=&MultiplyLinearColor(
     Linear Color=$Particles.Color
     Scale Factor=&MakeVector4(
      X=2
      Y=2
      Z=2
      W=&Multiply_Float(
       A=&Multiply_Float(
        A=&FloatFromCurve(
         FloatCurve=DataInterface() {propertyValues {Curve {keys [
             {interpMode=RCIM_Cubic, tangentMode=RCTM_User, time=0, value=0}
             {interpMode=RCIM_Cubic, tangentMode=RCTM_User, time=0.1, value=1}
             <at 0.8 1.125>
             {tangentMode=RCTM_User, time=1, value=0}
            ]}}}
         Scale Curve=0.5
        )
        B=&FNTimeOfDayScalar(Night Scale=0.25, Morning Scale=0.25, Day Scale=0.25, Evening Scale=0.25)
       )
       B=&ScalabilityDistanceBasedFloat(
        Curve:NiagaraDataInterfaceCurve=DataInterface() {propertyValues {Curve {keys [
            {interpMode=RCIM_Cubic, tangentMode=RCTM_User, time=0, value=0}
            {interpMode=RCIM_Cubic, tangentMode=RCTM_User, time=0.05, value=1}
            {interpMode=RCIM_Cubic, tangentMode=RCTM_User, time=0.7, value=1}
            <at 0.9 0>
           ]}}}
        CurveIndex:NiagaraFloat=$Emitter.Scalability.DistanceFraction
        Bias:NiagaraFloat=0
        Scale:NiagaraFloat=0.75
       )
      )
     )
    )) {isSetParameters=true, module=SetVariables_#1}
   &ScaleSpriteSize(Scale Factor=&Vector2DFromFloat(Value=&FloatFromCurve(
      FloatCurve=DataInterface() {propertyValues {Curve {keys [
          {arriveTangent=1.8751, interpMode=RCIM_Cubic, leaveTangent=1.8751, tangentMode=RCTM_User, time=0, value=0.2}
          {arriveTangent=1.0155, interpMode=RCIM_Cubic, leaveTangent=1.0155, tangentMode=RCTM_User, time=0.4, value=1.05}
          {arriveTangent=0.0421, interpMode=RCIM_Cubic, leaveTangent=0.0421, tangentMode=RCTM_User, time=1, value=1.25}
         ]}}}
      Scale Curve=&ScalabilityDistanceBasedFloat(
       Curve:NiagaraDataInterfaceCurve=DataInterface() {propertyValues {Curve {keys [<at 0 0.25>, {arriveTangent=0.5882, time=0.1, value=0.5}, <at 0.7 1>, <at 0.8 0>]}}}
       CurveIndex:NiagaraFloat=$Emitter.Scalability.DistanceFraction
       Bias:NiagaraFloat=0
       Scale:NiagaraFloat=1.5
      )
     )))
   &CurlNoiseForce(Noise Frequency=0.1, Noise Strength=500, Pan Noise=true, Pan Noise Field=<0 0 0.25>)
   &WindForce(Wind Speed=<150 0 0>)
   &SolveForcesAndVelocity(Clamp Velocity=false, Limit Acceleration=false)
   &SpriteRotationRate(Rotation Rate=&UniformRangedFloat(Minimum=-15, Maximum=15))
   &DynamicMaterialParameters(
    Index 0 Param 1=&FloatFromCurve(FloatCurve=DataInterface() {propertyValues {Curve {keys [
         {arriveTangent=-7.0802, interpMode=RCIM_Cubic, leaveTangent=-7.0802, tangentMode=RCTM_User, time=0, value=2}
         {interpMode=RCIM_Cubic, tangentMode=RCTM_User, time=0.5, value=0.05}
         {interpMode=RCIM_Cubic, tangentMode=RCTM_User, time=0.6, value=0.1}
         <at 1 2>
        ]}}})
    Index 0 Param 2=&FloatFromCurve(FloatCurve=DataInterface() {propertyValues {Curve {keys [
         {interpMode=RCIM_Cubic, tangentMode=RCTM_User, time=0, value=0}
         {interpMode=RCIM_Cubic, tangentMode=RCTM_User, time=0.5, value=1}
         {tangentMode=RCTM_User, time=1, value=0}
        ]}}})
    Index 0 Param 3=&RandomRangeFloat(Minimum=0.6, Maximum=0.8)
    Index 1 Param 1=150
    Module.Param0WriteEnabled=true
    Module.Param0WriteEnabled1=true
    Module.Param1WriteEnabled=true
    Module.Param1WriteEnabled1=true
    Module.Param2WriteEnabled=true
    Module.Param2WriteEnabled1=true
    Module.Param3WriteEnabled=true
    Module.Param3WriteEnabled1=true
    Write Parameter Index 1=true
   )
   &SubUVAnimation()
  ]
  rendererClasses [&NiagaraSpriteRendererProperties]
  renderers [{rendererClass=&NiagaraSpriteRendererProperties}]
 }
 Sparks {
  simTarget=GPUComputeSim
  emitterUpdateScript [
   &EmitterState(
    Enable Distance Culling=true
    Max Distance Response=::SleepAndClearParticles
    MaxDistance=20000
    MinDistance=250
    Module.bApplySpawnCountScaleByDistance=true
    Module.bUseMaxDistance=true
    Module.bUseMinDistance=false
    Reset Age On Awaken=true
    Scalability Mode=::NewEnumerator1
    Scale Spawn Count=true
    Spawn Count Scale By Distance=DataInterface() {propertyValues {
      bOptimizeLUT=false
      Curve {keys [<at 0 0.8>, <at 0.02 1>, <at 0.4 1.05>, {interpMode=RCIM_Cubic, tangentMode=RCTM_User, time=0.7, value=1}, <at 1 0>]}
     }}
   )
   &SpawnRate(SpawnRate=25, Use Spawn Probability=false)
  ]
  particleSpawnScript [
   /Niagara/Modules/Spawn/Initialization/InitializeParticle(
    Color=<rgba 15 5 0 1>
    Lifetime=&UniformRangedFloat(Minimum=3, Maximum=8)
    Module.Write Color=true
    Module.Write Lifetime=true
    Module.Write Mass=true
    Module.Write Material Random=true
    Module.Write Position=false
    Module.Write Scale=false
    Module.Write SpriteRotation=false
    Module.Write SpriteSize=true
    Position=&SimulationPosition()
    Sprite Size=&Vector2DFromFloat(Value=&RandomRangeFloat(Minimum=15, Maximum=35))
   )
   &SphereLocation(
    MaskSpawnGroup=false
    Module.Override Randomness=false
    Non Uniform Scale=&Multiply_VectorByFloat(
     Vector=&Multiply_VectorByFloat(Vector=<1 1 0.25>, Float=$User.GlobalScaler)
     Float=&ScalabilityDistanceBasedFloat(
      Curve:NiagaraDataInterfaceCurve=DataInterface() {propertyValues {Curve {keys [<at 0 0.25>, {arriveTangent=0.5882, time=0.1, value=0.5}, <at 0.85 1>, <at 0.95 0>]}}}
      CurveIndex:NiagaraFloat=$Emitter.Scalability.DistanceFraction
      Bias:NiagaraFloat=0
      Scale:NiagaraFloat=2
     )
    )
    Offset=<0 0 500>
    Sphere Radius=200
    Surface Only=false
   )
   &AddVelocity(Velocity=&UniformRangedVector(Minimum=<25 25 800>, Maximum=<-25 -25 1000>))
  ]
  particleUpdateScript [
   &ParticleState()
   &GravityForce(Gravity=<0 0 -250>)
   &Drag(
    Drag=&UniformRangedFloat(
     Minimum=0.1
     Maximum=&Subtract_Float(A=$Particles.NormalizedAge, B=&RandomRangeFloat(Minimum=0.1, Maximum=0.2))
    )
    Use Linear Drag=true
    Use Rotational Drag=false
   )
   &AccelerationForce()
   &CurlNoiseForce(
    Curl Noise Cone Mask Angle=150
    Curl Noise Cone Mask Falloff Angle=1
    Mask Curl Noise=true
    Noise Frequency=1
    Noise Strength=5000
    Pan Noise=true
    Pan Noise Field=<0 0 0.25>
   )
   &WindForce(Wind Speed=<500 0 0>, Wind Speed Scale=&RandomRangeFloat(Minimum=0.6, Maximum=1))
   &VortexForce(
    Influence Falloff Radius=5000
    Origin Pull Amount=2500
    Use Influence Falloff=true
    UseOriginOffset=true
    Vortex Force Amount=500
    Vortex Origin Offset:Vector3f=<1500 0 5000>
   )
   &SolveForcesAndVelocity(Clamp Velocity=true, Limit Acceleration=false, Speed Limit=2500)
   &ScaleSpriteSize(Scale Factor=&Vector2DFromFloat(Value=&FloatFromCurve(
      FloatCurve=DataInterface() {propertyValues {Curve {keys [{arriveTangent=0.0421, interpMode=RCIM_Cubic, leaveTangent=0.0421, tangentMode=RCTM_User, time=1, value=1.25}]}}}
      Scale Curve=&ScalabilityDistanceBasedFloat(
       Curve:NiagaraDataInterfaceCurve=DataInterface() {propertyValues {Curve {keys [<at 0 0.25>, {arriveTangent=0.5882, time=0.1, value=0.5}, <at 0.8 1>]}}}
       CurveIndex:NiagaraFloat=$Emitter.Scalability.DistanceFraction
       Bias:NiagaraFloat=0
       Scale:NiagaraFloat=1
      )
     )))
   &MaintainInCameraParticleScale(Half FOV=25)
   &ScaleSpriteSizeBySpeed(
    Max Scale Factor=<1 2>
    Min Scale Factor=<0.8 1>
    Sample Scale Factor By Curve=true
    Scale Factor Curve=DataInterface() {propertyValues {Curve {keys [
        {interpMode=RCIM_Cubic, tangentMode=RCTM_User, time=0, value=0}
        {arriveTangent=2.7141, interpMode=RCIM_Cubic, leaveTangent=2.7141, tangentMode=RCTM_User, time=1, value=1}
       ]}}}
    Velocity Threshold=2500
   )
   &ScaleColor(
    Scale Alpha=&FloatFromCurve(
     FloatCurve=DataInterface() {propertyValues {Curve {keys [<at 0 0 RCIM_Cubic>,<at 0.25 1>,<at 1.0011 0.002>]}}}
     Scale Curve=&Multiply_Float(
      A=&Waveform(
       Global Amplitude Scale=15
       "Waveform [1]"=::NewEnumerator0
       "[1] Phase Scale"=&RandomRangeFloat(Minimum=1, Maximum=2)
       "[1] Frequency"=&RandomRangeFloat(Minimum=0.1, Maximum=0.25)
       "[1] Amplitude Min/Max"=<0 1>
      )
      B=&Multiply_Float(
       A=&FloatFromCurve(
        FloatCurve=DataInterface() {propertyValues {Curve {keys [
            {interpMode=RCIM_Cubic, tangentMode=RCTM_User, time=0, value=0}
            {interpMode=RCIM_Cubic, tangentMode=RCTM_User, time=0.1, value=0.5}
            <at 0.8 1.125>
            {tangentMode=RCTM_User, time=1, value=1}
           ]}}}
        Scale Curve=&FNTimeOfDayScalar(Night Scale=1, Morning Scale=0.25, Evening Scale=0.25)
       )
       B=&ScalabilityDistanceBasedFloat(
        Curve:NiagaraDataInterfaceCurve=DataInterface() {propertyValues {Curve {keys [
            {interpMode=RCIM_Cubic, tangentMode=RCTM_User, time=0, value=0}
            {interpMode=RCIM_Cubic, tangentMode=RCTM_User, time=0.05, value=1}
            {interpMode=RCIM_Cubic, tangentMode=RCTM_User, time=0.7, value=1}
            <at 0.9 0>
           ]}}}
        CurveIndex:NiagaraFloat=$Emitter.Scalability.DistanceFraction
        Bias:NiagaraFloat=0
        Scale:NiagaraFloat=0.75
       )
      )
     )
    )
    Scale RGB=&MakeVectorFromLinearColorRGB(Color=&ColorFromCurve(ColorCurve=DataInterface() {propertyValues {
        AlphaCurve {keys []}
        BlueCurve {keys [<at 0 1.5 RCIM_Cubic>,<at 0.2 2.4556>,<at 0.6269 0.0752>,<at 1 0>]}
        GreenCurve {keys [<at 0 1.5 RCIM_Cubic>,<at 0.2 13.8501>,<at 0.6269 1.5411>,<at 1 0.1>]}
        RedCurve {keys [<at 0 1.5 RCIM_Cubic>,<at 0.2 30>,<at 0.6269 5>,<at 1 0.5>]}
       }}))
    ScaleA=true
    ScaleRGB=true
   )
  ]
  rendererClasses [&NiagaraSpriteRendererProperties]
  renderers [{rendererClass=&NiagaraSpriteRendererProperties}]
 }
}
summary {
 systemName=NS_CS_Fire
 emitters [
  {emitterName=FireAir, rendererClasses [&NiagaraSpriteRendererProperties]}
  {emitterName=Fire_BigEmbers, simTarget=GPUComputeSim, rendererClasses [&NiagaraSpriteRendererProperties]}
  {bEnabled=false, emitterName=Smoke, rendererClasses [&NiagaraSpriteRendererProperties]}
  {emitterName=Sparks, simTarget=GPUComputeSim, rendererClasses [&NiagaraSpriteRendererProperties]}
  {emitterName=FireGround, simTarget=GPUComputeSim, rendererClasses [&NiagaraSpriteRendererProperties]}
  {emitterName=Smoke001, rendererClasses [&NiagaraSpriteRendererProperties]}
 ]
 userVariables [User.GlobalScaler {type=NiagaraFloat, defaultValue=1, description=""}]
}
