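# Ambient GPU sprite system that scatters glowing purple ash and faint dust motes in the air around its owner.
# Both emitters spawn in a SphereScalar-scaled sphere, pass through a landscape-height module, and drift with gravity, drag and AirMovementScalar-driven curl noise.
# Sprites shrink near the camera; dust color and brightness follow NPC time of day, with alpha scaled by AlphaScalar.
# SphereScalar also drives spawn rate, while User.SpawnRateScale is unreferenced.
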NiagaraSystem /Game/Effects/Fort_Effects/Effects/Environment/CS/NS_CS_Motes_Ash
dependencies {
 usedDataInterfaces [
  &NiagaraDataInterfaceCamera
  &NiagaraDataInterfaceCurve
  &NiagaraDataInterfaceColorCurve
  /Script/Niagara.NiagaraDataInterfaceTexture
  &NiagaraDataInterfaceMeshRendererInfo
 ]
 usedDynamicInputs []
 usedModules [&SystemState]
}
emitters {
 Ash {
  simTarget=GPUComputeSim
  emitterUpdateScript [
   /SetVariables(Emitter.CameraPostion:Vector3f=&ReturnCameraVector(Camera Property=::NewEnumerator0, Coordinate Space=::Simulation)) {isSetParameters=true, module=SetVariables_#1}
   &EmitterState(
    Life Cycle Mode=::NewEnumerator1
    Loop Duration=&RandomRangeFloat(Minimum=5, Maximum=10)
    Recalculate Duration Each Loop=true
    UseLoopDelay=false
   )
   &SpawnRate(SpawnRate=&Multiply_Float(A=50, B=$User.SphereScalar), Use Spawn Probability=false)
  ]
  particleSpawnScript [
   &InitializeParticle(
    Color Maximum=<rgba 0.0443 0.0117 0.6875 1>
    Color Minimum=<rgba 0.5184 0 0.7865 1>
    Color Mode=::NewEnumerator2
    Lifetime Max=2.5
    Lifetime Mode=::NewEnumerator1
    Sprite Rotation Mode=::NewEnumerator2
    Sprite Size Mode=::NewEnumerator3
    Uniform Sprite Size=&Lerp_Float(
     A=3
     B=6
     Alpha=&FloatFromCurve(
      FloatCurve=DataInterface() {propertyValues {Curve {keys [
          {interpMode=RCIM_Cubic, tangentMode=RCTM_User, time=0, value=0}
          {arriveTangent=2.2019, interpMode=RCIM_Cubic, tangentMode=RCTM_Break, time=1, value=1}
         ]}}}
      CurveIndex=&RandomRangeFloat(Minimum=0, Maximum=1)
     )
    )
    UsePositionOffset=false
    Write Material Random=true
   )
   &AddVelocity(Velocity=&RandomVector(Vector Scale=&RandomRangeFloat(Minimum=0.5, Maximum=2)))
   /Niagara/Modules/Spawn/Location/SphereLocation(
    MaskSpawnGroup=false
    Module.Override Randomness=false
    Sphere Origin:NiagaraPosition=$Engine.Owner.Position
    Sphere Radius=&Multiply_Float(A=2500, B=$User.SphereScalar)
    Surface Only=false
   ) {module=SphereLocation001}
   /NiagaraFN/Modules/Spawn/FN_SpawnParticlesOnTerrainNearPlayer_ParticleType_V2(
    ColorTexture=DataInterface() {propertyValues {
      Texture=/Game/Effects/Fort_Effects/Textures/Environments/Apollo/T_WholeMap_Color_D_02
      TextureUserParameter {parameter {name=None, typeDefHandle {registeredTypeIndex=2}, varData []}}
     }}
    Height Texture -x -y:Vector3f=<-130048 -130048 0>
    HeightOffset:NiagaraFloat=&Lerp_Float(
     A=50
     B=1000
     Alpha=&FloatFromCurve(
      FloatCurve=DataInterface() {propertyValues {Curve {keys [
          {arriveTangent=0.0206, interpMode=RCIM_Cubic, leaveTangent=0.0206, tangentMode=RCTM_User, time=0, value=0}
          {arriveTangent=2.8989, interpMode=RCIM_Cubic, tangentMode=RCTM_Break, time=1, value=1}
         ]}}}
      CurveIndex=&RandomRangeFloat(Minimum=0, Maximum=1)
     )
    )
    HeightTexture +x +y:Vector3f=<130048 130048 0>
    Landscape:NiagaraDataInterfaceLandscape=$User.Landscape
    Player Position:NiagaraPosition=$Particles.Position
    Player Relative Particle Location:Vector2f=&UniformRangedVector2D(Minimum=<0 0>, Maximum=<0 0>)
    SandColor:LinearColor=<rgba 1 0 0 1>
    SnowColor:LinearColor=<rgba 0 0 1 1>
   )
   /Niagara/Modules/Update/Camera/MaintainInCameraParticleScale()
  ]
  particleUpdateScript [
   &ParticleState()
   &self:Camera_to_Particle_Distance_Calculation_01(Camera Postion:Vector3f=$Emitter.CameraPostion, Distance=2000, Particle Position:NiagaraPosition=$Particles.Position) {module=Camera_to_Particle_Distance_Calculation_01}
   &ScaleSpriteSize(Scale Factor=&Vector2DFromFloat(Value=&FloatFromCurve(
      FloatCurve=DataInterface() {propertyValues {Curve {keys [<at 0 0.25 RCIM_Cubic>,<at 1 1 RCIM_Cubic>]}}}
      CurveIndex=$Particles.NormalizedDistance
     )))
   &Drag(Drag=0.25, Use Linear Drag=true, Use Rotational Drag=false)
   &GravityForce(Gravity=&MakeVector(
     X=0
     Y=0
     Z=&FloatFromCurve(
      FloatCurve=DataInterface() {propertyValues {Curve {keys [<at 0 1 RCIM_Cubic>,<at 1 0>]}}}
      Scale Curve=&RandomRangeFloat(Minimum=-1, Maximum=-3)
     )
    ))
   &CurlNoiseForce(
    Curl Noise Cone Mask Angle=120
    Mask Curl Noise=true
    Noise Quality / Cost=::NewEnumerator0
    Noise Strength=&Multiply_Float(A=&RandomRangeFloat(Minimum=25, Maximum=100), B=$User.AirMovementScalar)
    Pan Noise=true
    Pan Noise Field=<0 0 1>
   )
   &ScaleColor(
    Scale Alpha=&FloatFromCurve(
     FloatCurve=DataInterface() {propertyValues {Curve {keys [
         {interpMode=RCIM_Cubic, tangentMode=RCTM_User, time=0, value=0}
         {arriveTangent=1.8182, leaveTangent=1.8182, time=0.1, value=1}
         {interpMode=RCIM_Cubic, tangentMode=RCTM_User, time=0.5, value=1}
         {interpMode=RCIM_Cubic, tangentMode=RCTM_Break, time=1, value=0}
        ]}}}
     Scale Curve=0.9
    )
    Scale RGB=&Multiply_VectorByFloat(Vector=<1 1 1>, Float=&RandomRangeFloat(Minimum=10, Maximum=25))
    ScaleA=true
    ScaleRGB=true
   )
   &SolveForcesAndVelocity(Clamp Velocity=false, Limit Acceleration=false)
   &ScaleSpriteSize(Uniform Curve Sprite Scale=DataInterface() {propertyValues {Curve {keys [
        {interpMode=RCIM_Cubic, tangentMode=RCTM_User, time=0, value=0}
        {interpMode=RCIM_Cubic, tangentMode=RCTM_User, time=0.1, value=1}
        {interpMode=RCIM_Cubic, tangentMode=RCTM_User, time=0.5, value=1}
        {interpMode=RCIM_Cubic, tangentMode=RCTM_Break, time=1, value=0}
       ]}}}) {module=ScaleSpriteSize001}
   &SubUVAnimation(SubUV Loop Count=0) {module=SubUVAnimation001}
  ]
  rendererClasses [&NiagaraSpriteRendererProperties]
  renderers [{rendererClass=&NiagaraSpriteRendererProperties}]
 }
 Dust_Motes {
  simTarget=GPUComputeSim
  emitterUpdateScript [
   /SetVariables(Emitter.CameraPostion:Vector3f=&ReturnCameraVector(Camera Property=::NewEnumerator0, Coordinate Space=::Simulation)) {isSetParameters=true, module=SetVariables_#1}
   &EmitterState(
    Life Cycle Mode=::NewEnumerator1
    Loop Duration=&RandomRangeFloat(Minimum=5, Maximum=10)
    Recalculate Duration Each Loop=true
    UseLoopDelay=false
   )
   &SpawnRate(SpawnRate=&Multiply_Float(A=75, B=$User.SphereScalar), Use Spawn Probability=false)
  ]
  particleSpawnScript [
   /Niagara/Modules/Spawn/Initialization/InitializeParticle(
    Color=&MakeLinearColorFromVectorAndFloat(
     "Vector (RGB)"=&MakeVectorFromLinearColorRGB(Color=&MultiplyLinearColor(
       Linear Color=&ColorFromCurve(
        ColorCurve=DataInterface() {propertyValues {
          AlphaCurve {keys []}
          BlueCurve {keys [<at 0.1189 0.75>,<at 0.1666 1>,<at 0.3109 0.75>,<at 0.5746 0.7083>,<at 0.8383 0.5>,<at 0.875 1>,<at 0.9354 1>,<at 1 0.75>]}
          GreenCurve {keys [<at 0.1189 0.0825>,<at 0.1666 0>,<at 0.3109 0.0825>,<at 0.5746 0.0701>,<at 0.8383 0.0206>,<at 0.875 0.0579>,<at 0.9354 0.045>,<at 1 0.0825>]}
          RedCurve {keys [<at 0.1189 0.1876>,<at 0.1666 0.9059>,<at 0.3109 0.1876>,<at 0.5746 0.2759>,<at 0.8383 0.2752>,<at 0.875 0.8121>,<at 0.9354 0.0997>,<at 1 0.1876>]}
         }}
        CurveIndex=Linked() {linkedVariable {name=NPC.FortniteNPC.TimeOfDay_0-1, type=NiagaraFloat}}
       )
       Scale Factor=&MakeVector4_FromFloat(Float=0.5)
      ))
     "Float (Alpha)"=&Multiply_Float(A=&RandomRangeFloat(Minimum=0.07, Maximum=0.13), B=$User.AlphaScalar)
    )
    Lifetime=&UniformRangedFloat(Minimum=3, Maximum=5)
    Module.Write Color=true
    Module.Write Lifetime=true
    Module.Write Mass=true
    Module.Write Material Random=true
    Module.Write Position=false
    Module.Write Scale=false
    Module.Write SpriteRotation=true
    Module.Write SpriteSize=true
    Position=&SimulationPosition()
    Sprite Rotation=&RandomRangeFloat(Minimum=0, Maximum=359)
    Sprite Size=&Vector2DFromFloat(Value=&RandomRangeFloat(Minimum=4, Maximum=7))
   )
   &InitializeParticle(
    Color=&MakeLinearColorFromVectorAndFloat(
     "Vector (RGB)"=&MakeVectorFromLinearColorRGB(Color=&MultiplyLinearColor(
       Linear Color=&ColorFromCurve(
        ColorCurve=DataInterface() {propertyValues {
          AlphaCurve {keys []}
          BlueCurve {keys [<at 0.1189 1>,<at 0.1666 0>,<at 0.3109 0.4062>,<at 0.5746 1>,<at 0.8383 0>,<at 0.875 0>,<at 0.9354 1>,<at 1 1>]}
          GreenCurve {keys [<at 0.1189 0.5833>,<at 0.1666 0.1067>,<at 0.3109 0.9>,<at 0.5746 1>,<at 0.8383 0.5>,<at 0.875 0.2>,<at 0.9354 0.2942>,<at 1 0.625>]}
          RedCurve {keys [<at 0.1189 0.5>,<at 0.1666 1>,<at 0.3109 1>,<at 0.5746 1>,<at 0.8383 1>,<at 0.875 1>,<at 0.9354 0.23>,<at 1 0.5>]}
         }}
        CurveIndex=Linked() {linkedVariable {name=NPC.FortniteNPC.TimeOfDay_0-1, type=NiagaraFloat}}
       )
       Scale Factor=&MakeVector4_FromFloat(Float=0.5)
      ))
     "Float (Alpha)"=&RandomRangeFloat(Minimum=0.07, Maximum=0.13)
    )
    Lifetime=&UniformRangedFloat(Minimum=3, Maximum=5)
    Module.Write Color=true
    Module.Write Lifetime=true
    Module.Write Mass=true
    Module.Write Material Random=true
    Module.Write Position=false
    Module.Write Scale=false
    Module.Write SpriteRotation=true
    Module.Write SpriteSize=true
    Position=&SimulationPosition()
    Sprite Rotation=&RandomRangeFloat(Minimum=0, Maximum=359)
    Sprite Size=&Vector2DFromFloat(Value=&RandomRangeFloat(Minimum=4, Maximum=7))
   ) {enabled=false, module=InitializeParticle001}
   &AddVelocity(Velocity=&RandomVector(Vector Scale=&RandomRangeFloat(Minimum=0.5, Maximum=2)))
   &SphereLocation(
    MaskSpawnGroup=false
    Module.Override Randomness=false
    Sphere Origin:NiagaraPosition=$Engine.Owner.Position
    Sphere Radius=&Multiply_Float(A=2500, B=$User.SphereScalar)
    Surface Only=false
   ) {module=SphereLocation001}
   &FN_SpawnParticlesOnTerrainNearPlayer_ParticleType_V2(
    ColorTexture=DataInterface() {propertyValues {
      Texture=&T_WholeMap_Color_D_02
      TextureUserParameter {parameter {name=None, typeDefHandle {registeredTypeIndex=2}, varData []}}
     }}
    Height Texture -x -y:Vector3f=<-130048 -130048 0>
    HeightOffset:NiagaraFloat=&Lerp_Float(
     A=50
     B=1000
     Alpha=&FloatFromCurve(
      FloatCurve=DataInterface() {propertyValues {Curve {keys [
          {arriveTangent=0.0206, interpMode=RCIM_Cubic, leaveTangent=0.0206, tangentMode=RCTM_User, time=0, value=0}
          {arriveTangent=2.8989, interpMode=RCIM_Cubic, tangentMode=RCTM_Break, time=1, value=1}
         ]}}}
      CurveIndex=&RandomRangeFloat(Minimum=0, Maximum=1)
     )
    )
    HeightTexture +x +y:Vector3f=<130048 130048 0>
    Landscape:NiagaraDataInterfaceLandscape=$User.Landscape
    Player Position:NiagaraPosition=$Particles.Position
    Player Relative Particle Location:Vector2f=&UniformRangedVector2D(Minimum=<0 0>, Maximum=<0 0>)
    SandColor:LinearColor=<rgba 1 0 0 1>
    SnowColor:LinearColor=<rgba 0 0 1 1>
   )
   &MaintainInCameraParticleScale()
   &DynamicMaterialParameters(
    Index 0 Param 1=2
    Index 0 Param 2=50
    Index 0 Param 3=200
    Module.Param0WriteEnabled=true
    Module.Param1WriteEnabled=true
    Module.Param2WriteEnabled=true
    Module.Param3WriteEnabled=true
   ) {module=DynamicMaterialParameters001}
  ]
  particleUpdateScript [
   &ParticleState()
   &self:Camera_to_Particle_Distance_Calculation_01(Camera Postion:Vector3f=$Emitter.CameraPostion, Distance=2000, Particle Position:NiagaraPosition=$Particles.Position) {module=Camera_to_Particle_Distance_Calculation_01}
   &ScaleSpriteSize(Scale Factor=&Vector2DFromFloat(Value=&FloatFromCurve(
      FloatCurve=DataInterface() {propertyValues {Curve {keys [<at 0 0.25 RCIM_Cubic>,<at 1 1.1 RCIM_Cubic>]}}}
      CurveIndex=$Particles.NormalizedDistance
     )))
   &Drag(Drag=0.25, Use Linear Drag=true, Use Rotational Drag=false)
   &GravityForce(Gravity=&MakeVector(
     X=0
     Y=0
     Z=&FloatFromCurve(
      FloatCurve=DataInterface() {propertyValues {Curve {keys [<at 0 1 RCIM_Cubic>,<at 1 0>]}}}
      Scale Curve=&RandomRangeFloat(Minimum=-1, Maximum=-3)
     )
    ))
   &CurlNoiseForce(
    Curl Noise Cone Mask Angle=120
    Mask Curl Noise=true
    Noise Quality / Cost=::NewEnumerator0
    Noise Strength=&Multiply_Float(A=&RandomRangeFloat(Minimum=10, Maximum=25), B=$User.AirMovementScalar)
    Pan Noise=true
    Pan Noise Field=<0 0 1>
   )
   &SpriteRotationRate(Rotation Rate=&RandomRangeFloat(Minimum=-45, Maximum=45))
   &ScaleColor(
    Scale Alpha=&FloatFromCurve(FloatCurve=DataInterface() {propertyValues {Curve {keys [
         {interpMode=RCIM_Cubic, tangentMode=RCTM_User, time=0, value=0}
         {arriveTangent=1.8182, leaveTangent=1.8182, time=0.35, value=1}
         <at 0.55 1>
         <at 1 0>
        ]}}})
    Scale RGB=&Multiply_VectorByFloat(
     Vector=<1 1 1>
     Float=/Niagara/DynamicInputs/Math/LerpMultipleFloats(
      Float Count:ENiagara_MultipleLerpCount=::NewEnumerator2
      1:NiagaraFloat=0.5
      2:NiagaraFloat=0.85
      3:NiagaraFloat=1
      4:NiagaraFloat=0.85
      5:NiagaraFloat=0.5
      Alpha:NiagaraFloat=Linked() {linkedVariable {name=NPC.FortniteNPC.TimeOfDay_0-1, type=NiagaraFloat}}
     )
    )
    ScaleA=true
    ScaleRGB=true
   )
   &SolveForcesAndVelocity(Clamp Velocity=false, Limit Acceleration=false)
  ]
  rendererClasses [&NiagaraSpriteRendererProperties]
  renderers [{rendererClass=&NiagaraSpriteRendererProperties}]
 }
}
summary {
 systemName=NS_CS_Motes_Ash
 emitters [
  {emitterName=Dust_Motes, simTarget=GPUComputeSim, rendererClasses [&NiagaraSpriteRendererProperties]}
  {emitterName=Ash, simTarget=GPUComputeSim, rendererClasses [&NiagaraSpriteRendererProperties]}
 ]
 userVariables [
  User.AirMovementScalar {type=NiagaraFloat, defaultValue=1, description=""}
  User.AlphaScalar {type=NiagaraFloat, defaultValue=1, description=""}
  User.Landscape {
   type=NiagaraDataInterfaceLandscape
   description=""
   defaultValue {
    struct=NiagaraExt_VariableValue_DataInterface
    value {
     dataInterface=&self:NiagaraDataInterfaceLandscape_0
     dataInterfaceClass=&NiagaraDataInterfaceLandscape
    }
   }
  }
  User.SpawnRateScale {type=NiagaraFloat, defaultValue=1, description=""}
  User.SphereScalar {type=NiagaraFloat, defaultValue=1, description=""}
 ]
}
