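# Niagara environment effect, named for Mother CS sparks, producing large magenta electrical arcs, glows and drifting sparks around a huge ring.
# A GPU Parent emitter bursts a few zero-size, short-lived points on a torus; Arcs, Light and Sparks spawn from them through particle attribute readers.
# Arcs are rotating, growing SubUV sprites, and Light is a camera-offset glow using FN_Light_v3.
# Sparks rise under curl noise and inverted gravity, with camera-maintained scale.
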NiagaraSystem /Game/Effects/Fort_Effects/Effects/Environment/CS/NS_Mother_CS_Sparks_01
dependencies {
 usedDataInterfaces [
  &NiagaraDataInterfaceParticleRead
  &NiagaraDataInterfaceCurve
  &NiagaraDataInterfaceArrayFloat
  &NiagaraDataInterfaceVectorCurve
 ]
 usedDynamicInputs [&MakeVectorFromLinearColorRGB]
 usedModules [&SystemState]
}
emitters {
 Arcs {
  simTarget=GPUComputeSim
  emitterUpdateScript [
   &EmitterState()
   &SpawnRate(Use Spawn Probability=false) {enabled=false}
   &SpawnParticlesFromOtherEmitter(
    Attribute Reader=DataInterface() {propertyValues {EmitterBinding {bindingMode=Other, emitterName=Parent}}}
    Spawn Rate=5
    Spawn Rate Per Particle Cap=5
    Use Maximum Age=false
    Use Minimum Age=false
   )
  ]
  particleSpawnScript [
   /Niagara/Modules/Spawn/Initialization/InitializeParticle(
    Color=<rgba 25 1 25 1>
    Lifetime=&RandomRangeFloat(Minimum=0.2, Maximum=0.25)
    Module.Write Color=true
    Module.Write Lifetime=true
    Module.Write Mass=true
    Module.Write Material Random=true
    Module.Write Position=false
    Module.Write Scale=false
    Module.Write SpriteRotation=true
    Module.Write SpriteSize=true
    Position=&SimulationPosition()
    Sprite Rotation=&RandomRangeFloat(Minimum=-180, Maximum=180)
    Sprite Size=&Vector2DFromFloat(Value=&RandomRangeFloat(Minimum=5000, Maximum=8000))
   )
   &SubUVAnimation(End Frame=3, SubUV Animation Mode:ENiagara_SubUVLookupMode=::NewEnumerator2)
   &SampleParticlesFromOtherEmitter()
  ]
  particleUpdateScript [
   &ParticleState()
   &CameraOffset(Camera Offset Amount=250)
   &ScaleSpriteSize(Scale Factor=&Vector2DFromFloat(Value=&FloatFromCurve(
      FloatCurve=DataInterface() {propertyValues {Curve {keys [
          {arriveTangent=1.7865, interpMode=RCIM_Cubic, leaveTangent=1.7865, tangentMode=RCTM_User, time=0.0023, value=0.1}
          {arriveTangent=0.6344, interpMode=RCIM_Cubic, leaveTangent=0.6344, tangentMode=RCTM_User, time=1, value=1}
         ]}}}
      Scale Curve=&SelectFloatFromArray(Direct Array Index=0, Float Selection Array=DataInterface() {propertyValues {FloatData [1,1,1,1.5]}})
     )))
   &SpriteRotationRate(Rotation Rate=100)
   &DynamicMaterialParameters(
    Index 0 Param 1=&RandomRangeFloat(Minimum=0, Maximum=1)
    Index 0 Param 3=&FloatFromCurve(FloatCurve=DataInterface() {propertyValues {Curve {keys [<at 0 0>,<at 1 0.1>]}}})
    Module.Param0WriteEnabled=true
    Module.Param1WriteEnabled=true
    Module.Param2WriteEnabled=true
    Module.Param3WriteEnabled=true
   )
   &ScaleColor(ScaleA=true, ScaleRGB=true)
  ]
  rendererClasses [&NiagaraSpriteRendererProperties]
  renderers [{rendererClass=&NiagaraSpriteRendererProperties}]
 }
 Light {
  simTarget=GPUComputeSim
  emitterUpdateScript [
   &EmitterState(Reset Age On Awaken=true, Scalability Mode=::NewEnumerator1)
   &SpawnBurst_Instantaneous(Spawn Count=1, Use Spawn Probability=false) {enabled=false}
   &SpawnParticlesFromOtherEmitter(
    Attribute Reader=DataInterface() {propertyValues {EmitterBinding {bindingMode=Other, emitterName=Parent}}}
    Spawn Rate=5
    Spawn Rate Per Particle Cap=1
    Use Maximum Age=true
    Use Minimum Age=true
   )
  ]
  particleSpawnScript [
   &InitializeParticle(
    Color=<rgba 4.0101 1.0075 5 1>
    Lifetime=&RandomRangeFloat(Minimum=0.2, Maximum=0.3)
    Module.Write Color=true
    Module.Write Lifetime=true
    Module.Write Mass=true
    Module.Write Material Random=true
    Module.Write Position=false
    Module.Write Scale=false
    Module.Write SpriteRotation=false
    Module.Write SpriteSize=true
    Position=&SimulationPosition()
    Sprite Size=&Vector2DFromFloat(Value=2)
   )
   &SampleParticlesFromOtherEmitter(Emitter Sampling Mode:ENiagara_EmitterLocSamplingMode=::NewEnumerator0)
  ]
  particleUpdateScript [
   &ParticleState()
   &CameraOffset(Camera Offset Amount=15000)
   &FN_Light_v3(
    Color=&LerpLinearColors(
     StartColor=<rgba 4.1155 2.6281 15 1>
     EndColor=<rgba 10 1.9712 1.4403 1>
     LerpFactor=&SelectFloatFromArray(Direct Array Index=0, Float Selection Array=DataInterface() {propertyValues {FloatData [0,0,0,1]}})
    )
    Create Color Attribute=true
    Create Exponent Attribute=true
    Create Radius Attribute=true
    Exponent=5
    Radius:NiagaraFloat=&Add_Float(
     A=75
     B=&FloatFromCurve(
      FloatCurve=DataInterface() {propertyValues {Curve {keys [<at 0 0>,<at 0.0579 0.8067>,<at 0.1401 0.355>,<at 0.2024 0.5966>,<at 0.3068 0.0924>,<at 0.4135 0.7857>,<at 0.4601 0.2185>,<at 0.5446 0.5126>,<at 0.7001 0.0819>,<at 0.7713 0.8382>,<at 0.9024 0.25>,<at 1 0>]}}}
      Scale Curve=&SelectFloatFromArray(Direct Array Index=0, Float Selection Array=DataInterface() {propertyValues {FloatData [0,15,22,35]}})
     )
    )
    Scale Color:NiagaraFloat=&Add_Float(
     A=1
     B=&FloatFromCurve(
      FloatCurve=DataInterface() {propertyValues {Curve {keys [<at 0 0>,<at 0.0579 0.8067>,<at 0.1401 0.355>,<at 0.2024 0.5966>,<at 0.3068 0.0924>,<at 0.4135 0.7857>,<at 0.4601 0.2185>,<at 0.5446 0.5126>,<at 0.7001 0.0819>,<at 0.7713 0.8382>,<at 0.9024 0.25>,<at 1 0>]}}}
      Scale Curve=&SelectFloatFromArray(Direct Array Index=0, Float Selection Array=DataInterface() {propertyValues {FloatData [0,0.25,0.5,1]}})
     )
    )
   )
   &ScaleColor(
    Scale Mode=::NewEnumerator1
    Scale RGBA=&MakeVector4_FromLinearColor(LinearColor=&LerpLinearColors(
      StartColor=<rgba 15 1.4431 7.312 0.1>
      EndColor=<rgba 20 0 13.7286 0.1>
      LerpFactor=&SelectFloatFromArray(Direct Array Index=0, Float Selection Array=DataInterface() {propertyValues {FloatData [0,0,0,1]}})
     ))
    ScaleRGBA=true
   )
   &ScaleSpriteSize(
    Uniform Curve Scale=25000
    Uniform Curve Sprite Scale=DataInterface() {propertyValues {Curve {keys [
        {arriveTangent=9.0725, interpMode=RCIM_Cubic, leaveTangent=9.0725, tangentMode=RCTM_User, time=0, value=0}
        {arriveTangent=-2.5249, interpMode=RCIM_Cubic, leaveTangent=-2.5249, tangentMode=RCTM_User, time=0.1408, value=0.9847}
        {arriveTangent=-0.0405, interpMode=RCIM_Cubic, leaveTangent=-0.0405, tangentMode=RCTM_User, time=1, value=0}
       ]}}}
   )
   &DynamicMaterialParameters(
    Index 0 Param 1=0.5
    Index 0 Param 2=1
    Index 0 Param 3=1
    Index 0 Param 4=1
    Module.Param0WriteEnabled=true
    Module.Param1WriteEnabled=true
    Module.Param2WriteEnabled=true
    Module.Param3WriteEnabled=true
   )
  ]
  rendererClasses [&NiagaraSpriteRendererProperties]
  renderers [{rendererClass=&NiagaraSpriteRendererProperties}]
 }
 Parent {
  simTarget=GPUComputeSim
  emitterUpdateScript [
   &EmitterState()
   &SpawnBurst_Instantaneous(Spawn Count=5, Use Spawn Probability=false)
   &SpawnRate(SpawnRate=250, Use Spawn Probability=false) {enabled=false}
  ]
  particleSpawnScript [
   &InitializeParticle(
    Lifetime=&UniformRangedFloat(Minimum=1, Maximum=1)
    Module.Write Color=true
    Module.Write Lifetime=true
    Module.Write Mass=true
    Module.Write Material Random=true
    Module.Write Position=false
    Module.Write Scale=false
    Module.Write SpriteRotation=true
    Module.Write SpriteSize=true
    Position=&SimulationPosition()
    Sprite Rotation=&UniformRangedFloat(Minimum=0, Maximum=360)
    Sprite Size=&Vector2DFromFloat(Value=&UniformRangedFloat(Minimum=0, Maximum=0))
   )
   &TorusLocation(
    Align Torus With Axis=false
    Handle Radius=5000
    Large Radius=190000
    MaskSpawnGroup=false
    Module.Override Randomness=false
    U Distribution=0.6
   )
  ]
  particleUpdateScript [
   &ParticleState()
   &SolveForcesAndVelocity(Clamp Velocity=false, Limit Acceleration=false)
  ]
  rendererClasses [&NiagaraSpriteRendererProperties]
  renderers [{rendererClass=&NiagaraSpriteRendererProperties}]
 }
 Sparks {
  simTarget=GPUComputeSim
  emitterUpdateScript [
   &EmitterState()
   &SpawnRate(SpawnRate=1000, Use Spawn Probability=false) {enabled=false}
   &SpawnParticlesFromOtherEmitter(
    Attribute Reader=DataInterface() {propertyValues {EmitterBinding {bindingMode=Other, emitterName=Parent}}}
    Minimum Age Spawn Threshold=0.1
    Spawn Rate=&RandomRangeFloat(Minimum=0, Maximum=5)
    Use Maximum Age=true
    Use Minimum Age=true
   )
  ]
  particleSpawnScript [
   &InitializeParticle(
    Color=<rgba 25 0 25 1>
    Lifetime=&RandomRangeFloat(Minimum=2, Maximum=3)
    Module.Write Color=true
    Module.Write Lifetime=true
    Module.Write Mass=true
    Module.Write Material Random=true
    Module.Write Position=true
    Module.Write Scale=false
    Module.Write SpriteRotation=false
    Module.Write SpriteSize=true
    Position=&SimulationPosition()
    Sprite Size=<250 1000>
   )
   &AddVelocityFromPoint(Module.Use Falloff Distance=false, Origin Offset=<0 0 -1>, Velocity Strength=500)
   &SampleParticlesFromOtherEmitter(Emitter Sampling Mode:ENiagara_EmitterLocSamplingMode=::NewEnumerator0)
   &SphereLocation(
    MaskSpawnGroup=false
    Override Seed=false
    Sphere Radius=500
    Surface Only=false
    UseNonUniformScale=false
    UseOffset=false
    UseSphereOrientation=false
   )
   &AddVelocity(Velocity=<0 0 5000>)
   &AddVelocityInCone(Cone Angle=15, Cone Axis=<1 0 1>, Module.Use Velocity Falloff On Cone Axis=false, Velocity Strength=5000)
  ]
  particleUpdateScript [
   &ParticleState()
   /Niagara/Modules/Update/Forces/CurlNoiseForce(Noise Frequency=1, Noise Strength=125000, Pan Noise=false, Pan Noise Field=<0 0 1>)
   &Drag(Drag=0, Use Linear Drag=true, Use Rotational Drag=false)
   &GravityForce(Gravity=<0 0 15000>)
   &SolveForcesAndVelocity(Clamp Velocity=true, Limit Acceleration=false, Speed Limit=10000)
   &ScaleSpriteSizeBySpeed(Max Scale Factor=<1 1>, Min Scale Factor=<1 0.5>)
   &ScaleColor(
    Scale RGB=&VectorFromCurve(
     VectorCurve=DataInterface() {propertyValues {XCurve {keys [<at 0 1>,<at 1 0>]}, YCurve {keys [<at 0 1>,<at 1 0>]}, ZCurve {keys [<at 0 1>,<at 1 0>]}}}
     Scale Curve=<15 1 25>
    )
    ScaleA=true
    ScaleRGB=true
   )
   /Niagara/Modules/Update/Camera/MaintainInCameraParticleScale(Half FOV=35)
  ]
  rendererClasses [&NiagaraSpriteRendererProperties]
  renderers [{rendererClass=&NiagaraSpriteRendererProperties}]
 }
}
summary {
 systemName=NS_Mother_CS_Sparks_01
 emitters [
  {emitterName=Light, simTarget=GPUComputeSim, rendererClasses [&NiagaraSpriteRendererProperties]}
  {emitterName=Arcs, simTarget=GPUComputeSim, rendererClasses [&NiagaraSpriteRendererProperties]}
  {emitterName=Sparks, simTarget=GPUComputeSim, rendererClasses [&NiagaraSpriteRendererProperties]}
  {emitterName=Parent, simTarget=GPUComputeSim, rendererClasses [&NiagaraSpriteRendererProperties]}
 ]
 userVariables []
}
